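; Blueprint actor that defines the position and radius of an asphalt mask for landscape materials.
; It writes its root component's world location as a vector and its radius amount as a scalar into the MPC_Landscape parameter collection, with parameter names taken from variables.
; The same writes run in the construction script and on BeginPlay, so the values are set both in the editor and at game start.
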(event EventBeginPlay
  (Rendering|Material|SetVectorParameterValue "/Clyde/Athena/Environments/Landscape/MPC/MPC_Landscape.MPC_Landscape" (Variables|Default|GetAsphaultMask) (Math|Conversions|ToLinearColor(Vector) (Transformation|GetWorldLocation (Variables|Default|GetDefaultSceneRoot))) :!class /Script/Engine.KismetMaterialLibrary)
  (Rendering|Material|SetScalarParameterValue "/Clyde/Athena/Environments/Landscape/MPC/MPC_Landscape.MPC_Landscape" (Variables|Default|GetAsphaultRadius) (Variables|Default|GetAsphaultRadiusAmount) :!class /Script/Engine.KismetMaterialLibrary))

(fn ConstructionScript
  (Rendering|Material|SetVectorParameterValue "/Clyde/Athena/Environments/Landscape/MPC/MPC_Landscape.MPC_Landscape" (Variables|Default|GetAsphaultMask) (Math|Conversions|ToLinearColor(Vector) (Transformation|GetWorldLocation (Variables|Default|GetDefaultSceneRoot))) :!class /Script/Engine.KismetMaterialLibrary)
  (Rendering|Material|SetScalarParameterValue "/Clyde/Athena/Environments/Landscape/MPC/MPC_Landscape.MPC_Landscape" (Variables|Default|GetAsphaultRadius) (Variables|Default|GetAsphaultRadiusAmount) :!class /Script/Engine.KismetMaterialLibrary))
