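# MF_Common_Asphalt is a material function that supplies the common asphalt surface for the Apollo landscape biome materials.
# It wraps MF_ApolloAsphalt and passes its Result through as the Material output.
# It also breaks BaseColor out of that result as its own output and exposes the asphalt's CrackHeight as Height.
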MaterialFunction /Clyde/Athena/Apollo/Environments/Landscape/Materials/Biome/CommonMaterials/MF_Common_Asphalt
graph {nodes [
  n1 = MaterialFunctionCall!(TilingSize) {materialFunction=/Clyde/Athena/Apollo/Environments/Landscape/Materials/MaterialFunctions/MF_ApolloAsphalt}
  FunctionOutput(None=@n1.Result) {outputName=Material, sortPriority=1}
  FunctionOutput(None=GetMaterialAttributes(None=@n1.Result) {attributeGetTypes [#1]}.BaseColor) {outputName=BaseColor}
  FunctionOutput(None=@n1.CrackHeight) {outputName=Height, sortPriority=2}
 ]}
