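# Material function that blends an asphalt material into incoming material attributes within a spherical area of the landscape biome materials.
# A SphereMask compares world position against a center and radius read from the MPC_Landscape parameter collection (AsphaultMask, AsphaultRadius).
# The mask serves as the alpha that blends the input attributes toward the output of MF_Common_Asphalt.
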MaterialFunction /Clyde/Athena/Apollo/Environments/Landscape/Materials/Biome/CommonMaterials/MF_AsphaultMask
graph {nodes [
  n1 = WorldPosition()
  n2 = ComponentMask(None=CollectionParameter() {collection=/Clyde/Athena/Environments/Landscape/MPC/MPC_Landscape, parameterName=AsphaultMask}) {b=true, g=true, r=true}
  n3 = CollectionParameter() {collection=&MPC_Landscape, parameterName=AsphaultRadius}
  FunctionOutput(None=BlendMaterialAttributes(
    A=FunctionInput() {inputType=FunctionInput_MaterialAttributes}
    B=MaterialFunctionCall() {materialFunction=&/MF_Common_Asphalt}.Material
    Alpha=SphereMask(A=@n1.XYZ, B=@n2, Radius=@n3)
   ))
 ]}
