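# Material function that supplies Apollo terrain biome colors and masks, with separate Mountain, Swamp, Farm and Forest channels.
# A static bool input picks landscape layer coordinates or world-position terrain coordinates to sample the shared landscape color and biome mask textures.
# The OverrideBiomeTexture static switch substitutes one custom quad-packed texture, using half-scaled, offset UVs to read colors from bottom left and masks from top right.
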MaterialFunction /Clyde/Athena/Apollo/Environments/Landscape/Materials/Biome/CommonMaterials/MF_ApolloTerrainMasks
graph {nodes [
  n1 = Add(A=NamedRerouteUsage(), B=Constant2Vector() {g=0.5})
  n2 = NamedRerouteUsage()
  n3 = StaticSwitch(
   True=LandscapeLayerCoords() {mappingScale=2033}
   False=MaterialFunctionCall!(WorldPosition) {materialFunction=/Clyde/Athena/Environments/Landscape/MaterialFunctions/Utilities/MF_Apollo_Terrain_Coords}.Result
   Value=FunctionInput(Preview=StaticBool()) {bUsePreviewValueAsDefault=true, inputName=UseLandscapeCoords, inputType=FunctionInput_StaticBool}
  )
  n4 = TextureSample(UVs=@n3) {
   samplerSource=SSM_Wrap_WorldGroupSettings
   texture=&/T_ApolloLandscapeMask
  }
  FunctionOutput(None=StaticSwitchParameter(True=TextureSample(UVs=@n1, Tex=@n2).RGB, False=@n4.RGB) {group=BomeSettings, parameterName=OverrideBiomeTexture}) {outputName=BiomeColors}
  n5 = TextureSample(UVs=Add(A=NamedRerouteUsage(), B=Constant2Vector() {r=0.5}), Tex=NamedRerouteUsage())
  n6 = TextureSample(UVs=@n3) {
   samplerType=SAMPLERTYPE_LinearColor
   texture=&/T_ApolloLandscapeBiomeMask
  }
  FunctionOutput(None=StaticSwitchParameter(True=@n5.RGB, False=@n6.RGB) {group=BomeSettings, parameterName=OverrideBiomeTexture}) {outputName=BiomeMaskRGBA}
  FunctionOutput(None=StaticSwitchParameter(True=@n5.R, False=@n6.R) {group=BomeSettings, parameterName=OverrideBiomeTexture}) {outputName=Mountain}
  FunctionOutput(None=StaticSwitchParameter(True=@n5.G, False=@n6.G) {group=BomeSettings, parameterName=OverrideBiomeTexture}) {outputName=Swamp}
  FunctionOutput(None=StaticSwitchParameter(True=@n5.B, False=@n6.B) {group=BomeSettings, parameterName=OverrideBiomeTexture}) {outputName=Farm}
  FunctionOutput(None=StaticSwitchParameter(True=@n5.A, False=@n6.A) {group=BomeSettings, parameterName=OverrideBiomeTexture}) {outputName=Forest}
  n7 = TextureObjectParameter() {
   desc="Top right is biome mask, bottom left is grass color"
   group=BomeSettings
   parameterName="CustomBiomeTexture(Quad)"
   samplerType=SAMPLERTYPE_Masks
   texture=&T_BlackMask
  }
  NamedRerouteDeclaration(None=@n7) {name=BiomeTexture, nodeColor=<rgba 1 0 0.0469 1>}
  NamedRerouteDeclaration(None=Multiply(A=@n3) {constB=0.5}) {name=BaseCoord, nodeColor=<rgba 0.6875 1 0 1>}
 ]}
