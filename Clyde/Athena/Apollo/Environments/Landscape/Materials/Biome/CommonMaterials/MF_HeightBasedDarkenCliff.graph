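# Material function for Apollo landscape biome materials that computes a height-based factor, named for darkening cliffs.
# It inverts world-space Z, then offsets, scales and saturates it with fixed constants, so the value falls as height increases.
# The optional DarknessAmount input is added to that gradient, setting its minimum, and the sum is saturated again.
# A second output, WorldZ, exposes the raw height.
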MaterialFunction /Clyde/Athena/Apollo/Environments/Landscape/Materials/Biome/CommonMaterials/MF_HeightBasedDarkenCliff
graph {nodes [
  n1 = ComponentMask(None=WorldPosition().XYZ) {b=true}
  n2 = OneMinus(None=@n1)
  n3 = Constant() {r=10000}
  n4 = Saturate(None=Divide(A=Add(A=@n2, B=@n3), B=Constant() {r=30000}))
  n5 = FunctionInput(Preview=Constant() {r=0.35}) {bUsePreviewValueAsDefault=true, inputName=DarknessAmount, inputType=FunctionInput_Scalar}
  FunctionOutput(None=Saturate(None=Add(A=@n4, B=@n5)))
  FunctionOutput(None=@n1) {outputName=WorldZ}
 ]}
