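# Landscape material function that builds the Apollo biome cliff layer, outputting material attributes and a height map.
# Two world-aligned cliff texture sets height-blend through large-scale swirl noise, then height-blend with a grey-colored dirt layer by the Slope input.
# A world-XY rock detail texture modulates color and specular, while a height-darkening function and world-Z gradient scale color and roughness.
# The biome color and mask inputs are declared but unconnected.
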MaterialFunction /Clyde/Athena/Apollo/Environments/Landscape/Materials/Biome/CommonMaterials/MF_Common_Cliff
graph {nodes [
  FunctionInput(Preview=Constant() {r=1}) {bUsePreviewValueAsDefault=true, inputName=BiomeColorMask, inputType=FunctionInput_Scalar, sortPriority=1}
  FunctionInput() {bUsePreviewValueAsDefault=true, inputName=BiomeColor, sortPriority=3, previewValue=<1 1 1 1>}
  FunctionInput() {bUsePreviewValueAsDefault=true, inputName=BiomeColorDiv, sortPriority=2, previewValue=<1 1 1 1>}
  FunctionInput(Preview=StaticBool()) {bUsePreviewValueAsDefault=true, inputName=UseBiomeColorMask, inputType=FunctionInput_StaticBool, previewValue=<64 64 64 6>}
  n1 = Constant() {r=3950}
  n2 = MaterialFunctionCall!(DarknessAmount) {materialFunction=&/MF_HeightBasedDarkenCliff}
  n3 = Saturate(None=Divide(A=Add(A=@n2.WorldZ, B=@n1), B=Constant() {r=250}))
  n4 = Max(A=@n3, B=Constant() {r=0.5})
  n5 = MaterialFunctionCall!(
   Diffuse Texture Object=TextureObject() {texture=/Game/Environments/Apollo/Sets/Terrain/Textures/T_Apollo_Cliff_D_03}
   Spec Texture Object=TextureObject() {samplerType=SAMPLERTYPE_Masks, texture=/Game/Environments/Apollo/Sets/Terrain/Textures/T_Apollo_Cliff_S_03}
   TextureSize=Constant() {r=2500}
   --------------------------
   Use seperate settings for Z projection
   Use seperate texture for Z projection
   Texture Size Z projection
   Z Plane Diffuse Texture Object
   Z Plane Spec Texture Object
   "--------------------------      "
   ProjectionTransitionContrast=Constant() {r=0.5}
   Use Texture Alpha Heightmap to Refine the Transistion=StaticBool() {value=true}
   HeightLerpTransitionContrast=Constant() {r=0.5}
   PickTheDiffuseHeightChannel
   TheDiffuseHeightChannel
   "--------------------------       "
   NormalTextureObject=TextureObject() {samplerType=SAMPLERTYPE_Normal, texture=/Game/Environments/Apollo/Sets/Terrain/Textures/T_Apollo_Cliff_N_03}
   World Space Normals=StaticBool()
   "--------------------------  "
   Export Float 4 Diffuse=StaticBool() {value=true}
   Export Float 4 Spec
   "--------------------------     "
   WorldPosition=WorldPosition() {worldPositionShaderOffset=WPT_ExcludeAllShaderOffsets}.XYZ
   World Space Normal
  ) {materialFunction=/Engine/Functions/Engine_MaterialFunctions02/Texturing/WorldAlignedTextures_Complex}
  n6 = SetMaterialAttributes!(
   MaterialAttributes
   BaseColor=ComponentMask(None=@n5.XY Diffuse Texture) {b=true, g=true, r=true}
   Specular=ComponentMask(None=@n5.XY Spec) {r=true}
   Roughness=ComponentMask(None=@n5.XY Spec) {b=true}
   Normal=@n5.Normal XY Texture
  ) {attributeSetTypes [
    #1
    #2
    #3
    #4
   ]}
  n7 = TextureSample(UVs=MaterialFunctionCall!(Scale=Constant() {r=-15000}) {materialFunction=/Engine/Functions/Engine_MaterialFunctions01/Coordinates/WorldPosition-XY}.Result) {
   samplerSource=SSM_Wrap_WorldGroupSettings
   samplerType=SAMPLERTYPE_LinearGrayscale
   texture=/Game/Global/Textures/Default/Noise/T_SwirlNoise_N_height
  }
  n8 = MaterialFunctionCall(
   Diffuse Texture Object=TextureObject() {texture=/Game/Environments/Apollo/Sets/Terrain/Textures/T_Apollo_Cliff_D_01}
   Spec Texture Object=TextureObject() {samplerType=SAMPLERTYPE_Masks, texture=/Game/Environments/Apollo/Sets/Terrain/Textures/T_Apollo_Cliff_S_01}
   TextureSize=Constant3Vector() {constant=<rgba 3000 3000 2000 1>}
   ProjectionTransitionContrast=Constant() {r=1}
   Use Texture Alpha Heightmap to Refine the Transistion=StaticBool() {value=true}
   HeightLerpTransitionContrast=Constant() {r=5}
   NormalTextureObject=TextureObject() {samplerType=SAMPLERTYPE_Normal, texture=/Game/Environments/Apollo/Sets/Terrain/Textures/T_Apollo_Cliff_N_01}
   World Space Normals=StaticBool()
   Export Float 4 Diffuse=StaticBool() {value=true}
   WorldPosition=WorldPosition() {worldPositionShaderOffset=WPT_ExcludeAllShaderOffsets}.XYZ
  ) {materialFunction=&WorldAlignedTextures_Complex}
  n9 = MaterialFunctionCall!(
   Diffuse 1
   Diffuse 2
   Transistion Phase=@n7.RGB
   Height Texture 1=ComponentMask(None=@n8.XY Diffuse Texture) {a=true}
   Height Texture 2=ComponentMask(None=@n5.XY Diffuse Texture) {a=true}
   Contrast=Constant() {r=2}
  ) {materialFunction=/Engine/Functions/Engine_MaterialFunctions02/Texturing/HeightLerpWithTwoHeightMaps}
  n10 = SetMaterialAttributes(
   BaseColor=ComponentMask(None=@n8.XY Diffuse Texture) {b=true, g=true, r=true}
   Specular=ComponentMask(None=@n8.XY Spec) {r=true}
   Roughness=ComponentMask(None=@n8.XY Spec) {b=true}
   Normal=@n8.Normal XY Texture
  ) {attributeSetTypes [
    #1
    #2
    #3
    #4
   ]}
  n11 = Multiply(A=LandscapeLayerCoords() {mappingScale=1}, B=Constant() {r=0.065})
  n12 = TextureSample(UVs=@n11) {
   samplerSource=SSM_Wrap_WorldGroupSettings
   samplerType=SAMPLERTYPE_Masks
   texture=/Game/Environments/Apollo/Sets/Terrain/Textures/T_Apollo_Dirt_S_02
  }
  n13 = Constant() {r=1.1}
  n14 = TextureSample(UVs=@n11) {
   samplerSource=SSM_Wrap_WorldGroupSettings
   samplerType=SAMPLERTYPE_Normal
   texture=/Game/Environments/Apollo/Sets/Terrain/Textures/T_Apollo_Dirt_N_02
  }
  n15 = SetMaterialAttributes(BaseColor=Constant() {r=0.25}, Specular=@n12.R, Roughness=Saturate(None=Multiply(A=@n12.B, B=@n13)), Normal=@n14.RGB) {attributeSetTypes [
    #1
    #2
    #3
    #4
   ]}
  n16 = Constant() {r=2.5}
  n17 = Constant() {r=-0.5}
  n18 = FunctionInput() {inputName=Slope, inputType=FunctionInput_Scalar, sortPriority=4}
  n19 = MaterialFunctionCall(
   Transistion Phase=Saturate(None=LinearInterpolate(A=@n16, B=@n17, Alpha=@n18))
   Height Texture 1=TextureSample(UVs=@n11) {samplerSource=SSM_Wrap_WorldGroupSettings, texture=/Game/Environments/Apollo/Sets/Terrain/Textures/T_Apollo_Dirt_D_02}.A
   Height Texture 2=@n9.Resulting Height Map
  ) {materialFunction=&HeightLerpWithTwoHeightMaps}
  n20 = GetMaterialAttributes(None=BlendMaterialAttributes(A=BlendMaterialAttributes(A=@n10, B=@n6, Alpha=@n9.Alpha), B=@n15, Alpha=@n19.Lerp Alpha No Contrast)) {attributeGetTypes [
    #4
    #3
    #2
    #1
   ]}
  n21 = Constant() {r=-5}
  n22 = Constant() {r=1.4}
  n23 = TextureSample(UVs=MaterialFunctionCall(Scale=Constant() {r=-1000}) {materialFunction=&WorldPosition-XY}.Result) {samplerSource=SSM_Wrap_WorldGroupSettings, texture=/Game/Environments/Apollo/Rocks/Natural_Rocks/Textures/T_RockDetails_02D}
  n24 = Add(A=ComponentMask(None=@n23.RGB) {r=true}, B=Constant() {r=-0.4})
  n25 = Saturate(None=LinearInterpolate(A=Constant() {r=-0.25}, B=Constant() {r=1.5}, Alpha=@n20.Specular))
  n26 = Constant() {r=0.5}
  n27 = Constant() {r=0.5}
  n28 = SetMaterialAttributes!(
   MaterialAttributes=@n20.MaterialAttributes
   Normal=@n20.MaterialAttributes
   Roughness=Saturate(None=LinearInterpolate(A=@n21, B=@n22, Alpha=@n20.Roughness))
   Specular=Multiply(A=Saturate(None=@n24), B=Clamp(None=@n25, Max=@n26) {clampMode=CMODE_ClampMax})
   BaseColor=Multiply(A=Power(Base=@n23.RGB, Exp=@n27), B=@n20.BaseColor)
  ) {attributeSetTypes [
    #4
    #3
    #2
    #1
   ]}
  n29 = GetMaterialAttributes(None=@n28) {attributeGetTypes [#1, #3]}
  n30 = Multiply(A=@n2.Result, B=@n29.BaseColor)
  FunctionOutput(None=SetMaterialAttributes!(MaterialAttributes=@n28, BaseColor=Multiply(A=@n4, B=@n30), Roughness=Multiply(A=@n3, B=@n29.Roughness)) {attributeSetTypes [#1, #3]}) {outputName=Material}
  FunctionOutput(None=@n19.Resulting Height Map) {outputName=Height}
  MaterialFunctionCall()
  Saturate()
 ]}
