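# Material function for the common grass layer of the Apollo landscape, outputting material attributes, heightmap, color masks and a simpler mobile base color.
# World-projected color masks at several scales build macro variation, which is divided by an average color, tinted by biome colors and blended to edge colors.
# Variation fades by RVT mip level, or camera distance without RVT.
# Detail textures add shading, MPC_Landscape drives brightness and desaturation, and CliffMask lowers height.
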MaterialFunction /Clyde/Athena/Apollo/Environments/Landscape/Materials/Biome/CommonMaterials/MF_Common_Grass
graph {nodes [
  Constant3Vector() {constant=<rgba 0.0378 0.11 0.0022 1>}
  Constant3Vector() {constant=<rgba 0.1199 0.22 0.0068 1>}
  Constant3Vector() {constant=<rgba 0.098 0.18 0.009 1>}
  Constant3Vector() {constant=<rgba 0.08 0.16 0.0032 1>}
  Constant3Vector() {constant=<rgba 0.0238 0.1 0 1>}
  Constant3Vector() {constant=<rgba 0.0708 0.16 0 1>}
  Constant3Vector() {constant=<rgba 0.0995 0.15 0.03 1>}
  Constant3Vector() {constant=<rgba 0.6751 0.9 0.135 1>}
  Constant() {r=3500}
  Constant() {r=1.5}
  Constant() {r=0.3}
  Constant() {r=1}
  Constant() {r=-0.6}
  Constant() {r=2}
  Constant() {r=0.75}
  Constant()
  Constant() {r=-3600}
  Constant() {r=1.5}
  Constant() {r=13000}
  Constant() {r=-9}
  Constant() {r=3000}
  Constant() {r=0.89}
  Constant() {r=500}
  Constant() {r=4560}
  Constant() {r=900}
  Constant() {r=0.55}
  Constant() {r=3}
  Constant() {r=1500}
  Constant() {r=0.5}
  Constant() {r=1600}
  Constant() {r=6}
  Constant() {r=0.85}
  Constant() {r=3}
  Constant() {r=12}
  Constant() {r=2}
  Constant()
  Constant() {r=5599}
  Constant() {r=1.5}
  Constant() {r=0.1}
  n1 = ScalarParameter() {defaultValue=950, group=Grass, parameterName=Grass-NormalScale}
  n2 = Divide(A=ComponentMask(None=WorldPosition() {worldPositionShaderOffset=WPT_ExcludeAllShaderOffsets}.XYZ) {g=true, r=true}, B=@n1)
  n3 = TextureSample(UVs=@n2) {samplerSource=SSM_Wrap_WorldGroupSettings, texture=/Game/Environments/Apollo/Sets/Terrain/Textures/T_Apollo_Grass_D_01}
  n4 = Multiply(A=Constant() {r=0.05}, B=@n3.G)
  n5 = Saturate(None=LinearInterpolate(A=Constant() {r=0.55}, B=Constant() {r=3}, Alpha=@n3.G))
  n6 = WorldPosition() {worldPositionShaderOffset=WPT_ExcludeAllShaderOffsets}
  n7 = TextureSample(UVs=Divide(A=ComponentMask(None=@n6.XYZ) {g=true, r=true}, B=ScalarParameter() {defaultValue=-13000, parameterName=fet654644})) {
   samplerSource=SSM_Wrap_WorldGroupSettings
   samplerType=SAMPLERTYPE_Grayscale
   texture=/Game/Athena/Apollo/Environments/Landscape/Materials/Biome/CommonMaterials/T_ApolloGrass_ColorMask
  }
  n8 = LinearInterpolate(
   A=VectorParameter() {parameterName=GrassLSLayerEdgeColorA, defaultValue=<rgba 0.18 0.1308 0.018 1>}.RGB
   B=VectorParameter() {parameterName=GrassLSLayerEdgeColorB, defaultValue=<rgba 0.2 0.1454 0.02 1>}.RGB
   Alpha=@n7.R
  )
  n9 = FunctionInput() {bUsePreviewValueAsDefault=true, inputName=Grass-EdgeBiomeColorA}
  n10 = FunctionInput() {bUsePreviewValueAsDefault=true, inputName=Grass-EdgeBiomeColorB}
  n11 = Constant()
  n12 = FunctionInput(Preview=StaticBool()) {bUsePreviewValueAsDefault=true, inputName=UseBiomeColorMask, inputType=FunctionInput_StaticBool, previewValue=<64 64 64 6>}
  n13 = StaticSwitch(
   True=LinearInterpolate(
    A=@n8
    B=LinearInterpolate(A=@n9, B=@n10, Alpha=@n7.R)
    Alpha=FunctionInput(Preview=@n11) {bUsePreviewValueAsDefault=true, inputName=BiomeMask, inputType=FunctionInput_Scalar}
   )
   False=@n8
   Value=@n12
  ) {desc=Z projection}
  n14 = FunctionInput() {bUsePreviewValueAsDefault=true, inputName=CombinedLandscapeLayersMask, inputType=FunctionInput_Scalar}
  n15 = MaterialFunctionCall!(UseLandscapeCoords) {materialFunction=&/MF_ApolloTerrainMasks}
  n16 = VectorParameter() {parameterName=GrassMacroColorA, defaultValue=<rgba 0.0378 0.11 0.0022 1>}
  n17 = Constant()
  n18 = Constant() {r=1.5}
  n19 = ComponentMask(None=WorldPosition() {worldPositionShaderOffset=WPT_ExcludeAllShaderOffsets}.XYZ) {g=true, r=true}
  n20 = TextureSample(UVs=Divide(A=@n19, B=Constant() {r=5599})) {
   samplerSource=SSM_Wrap_WorldGroupSettings
   samplerType=SAMPLERTYPE_Grayscale
   texture=&T_ApolloGrass_ColorMask
  }
  n21 = LinearInterpolate(
   A=VectorParameter() {parameterName=SSC-1, defaultValue=<rgba 0.08 0.16 0.0032 1>}.RGB
   B=VectorParameter() {parameterName=SSC-2, defaultValue=<rgba 0.0238 0.1 0 1>}.RGB
   Alpha=Saturate(None=LinearInterpolate(A=@n17, B=@n18, Alpha=@n20.A))
  )
  n22 = VectorParameter() {parameterName=SSC-3, defaultValue=<rgba 0.0708 0.16 0 1>}
  n23 = TextureSample(UVs=Divide(A=@n19, B=Constant() {r=-3600})) {
   samplerSource=SSM_Wrap_WorldGroupSettings
   samplerType=SAMPLERTYPE_Grayscale
   texture=&T_ApolloGrass_ColorMask
  }
  n24 = Saturate(None=LinearInterpolate(A=Constant() {r=1.5}, B=Constant() {r=-9}, Alpha=@n23.A))
  n25 = Constant() {r=500}
  n26 = TextureSample(UVs=Divide(A=Add(A=@n19, B=@n25), B=Constant() {r=4560})) {
   samplerSource=SSM_Wrap_WorldGroupSettings
   samplerType=SAMPLERTYPE_Grayscale
   texture=&T_ApolloGrass_ColorMask
  }
  n27 = Saturate(None=LinearInterpolate(A=Constant() {r=-0.6}, B=Constant() {r=2}, Alpha=@n26.A))
  n28 = Constant() {r=4}
  n29 = LinearInterpolate(
   A=@n16.RGB
   B=LinearInterpolate(
    A=LinearInterpolate(A=@n21, B=@n22.RGB, Alpha=@n24)
    B=VectorParameter() {parameterName=yellowish, defaultValue=<rgba 0.098 0.18 0.009 1>}.RGB
    Alpha=@n27
   )
   Alpha=Saturate(None=Multiply(A=@n28, B=@n7.R))
  )
  n30 = VectorParameter() {parameterName=GrassMacroColorB, defaultValue=<rgba 0.1199 0.22 0.0068 1>}
  n31 = LinearInterpolate(A=@n16.RGB, B=@n30.RGB, Alpha=@n7.R)
  n32 = ViewProperty() {property=MEVP_RuntimeVirtualTextureOutputLevel}
  n33 = ComponentMask(None=@n32.InvProperty) {r=true}
  n34 = ScalarParameter() {defaultValue=2.5, parameterName=RVT_MipDistanceNearEffect}
  n35 = LinearInterpolate(A=@n31, B=@n29, Alpha=Saturate(None=Multiply(A=@n33, B=@n34)))
  n36 = LinearInterpolate(A=@n16.RGB, B=@n30.RGB)
  n37 = ComponentMask(None=@n32.Property) {r=true}
  n38 = Saturate(None=Divide(A=@n37, B=ScalarParameter() {defaultValue=18, parameterName=RVT_MipDistanceFarEffect}))
  n39 = VectorParameter() {parameterName=GrassAveragedColor-Div, defaultValue=<rgba 0.0995 0.15 0.03 1>}
  n40 = Divide(A=VirtualTextureFeatureSwitch!(No=@n29, Yes=LinearInterpolate(A=@n35, B=@n36, Alpha=@n38)), B=@n39.RGB)
  n41 = Multiply(A=@n15.BiomeColors, B=@n40)
  n42 = WorldPosition()
  n43 = CameraPositionWS()
  n44 = Constant() {r=20000}
  n45 = Constant()
  n46 = CameraPositionWS()
  n47 = WorldPosition()
  n48 = Constant() {r=100000}
  n49 = Constant()
  n50 = LinearInterpolate(
   A=LinearInterpolate(A=@n31, B=@n29, Alpha=SphereMask(A=@n42.XYZ, B=@n43, Radius=@n44, Hardness=@n45))
   B=LinearInterpolate(A=@n16.RGB, B=@n30.RGB)
   Alpha=OneMinus(None=SphereMask(A=@n46, B=@n47.XYZ, Radius=@n48, Hardness=@n49))
  )
  n51 = Multiply(A=@n15.BiomeColors, B=Divide(A=@n50, B=@n39.RGB))
  n52 = Multiply(
   A=@n5
   B=StaticSwitchParameter(
    True=LinearInterpolate(A=@n41, B=@n13, Alpha=@n14)
    False=StaticSwitchParameter(True=@n51, False=@n50) {defaultValue=true, parameterName=UseBiomeColor}
   ) {defaultValue=true, parameterName=UseRVT}
  )
  n53 = CollectionParameter() {collection=/Clyde/Athena/Environments/Landscape/MPC/MPC_Landscape, parameterName=ClydeGrassBrighten}
  n54 = CollectionParameter() {collection=&MPC_Landscape, parameterName=ClydeGrassSaturation}
  n55 = Desaturation(None=Multiply(A=Add(A=@n4, B=@n52), B=@n53) {constB=1.3}, Fraction=@n54)
  FunctionOutput(None=@n55) {outputName=Basecolor, sortPriority=1}
  FunctionOutput(None=@n20.A) {outputName=GrassColorMaskA, sortPriority=8}
  n56 = Multiply(A=@n15.BiomeColors, B=VectorParameter() {parameterName=Mobile-StaticTerrainBaseColor, defaultValue=<rgba 0.6751 0.9 0.135 1>}.RGB)
  n57 = VectorParameter() {desc=Color to use instead of biome color, parameterName=Mobile-BiomeGrassColorOverride_KL, defaultValue=<rgba 0.6751 0.9 0.135 1>}
  n58 = ScalarParameter() {desc=Blends Biome Grass Color Override, parameterName=MobileOverrideBiomeGrassColor_KL}
  n59 = Add(A=@n4, B=Multiply(A=@n5, B=LinearInterpolate(A=@n56, B=@n57.RGB, Alpha=@n58)))
  FunctionOutput(None=Desaturation(None=Multiply(A=@n59, B=@n53) {constB=1.3}, Fraction=@n54)) {outputName=Mobile-Basecolor, sortPriority=1}
  FunctionOutput(None=@n27) {outputName=GrassColorMaskA3, sortPriority=10}
  n60 = TextureSample(UVs=@n2) {
   samplerSource=SSM_Wrap_WorldGroupSettings
   samplerType=SAMPLERTYPE_Masks
   texture=/Game/Environments/Apollo/Sets/Terrain/Textures/T_Apollo_Grass_S_01
  }
  n61 = OneMinus(None=@n27)
  n62 = Clamp(
   None=LinearInterpolate(
    A=ScalarParameter() {defaultValue=0.1, parameterName=Terrain-GrsDetSpecA}
    B=ScalarParameter() {defaultValue=1.5, parameterName=Terrain-GrsDetSpecB}
    Alpha=Multiply(A=@n60.R, B=@n61)
   )
   Max=Constant() {r=0.5}
  ) {clampMode=CMODE_ClampMax}
  n63 = ScalarParameter() {defaultValue=0.3, parameterName=Terrain-GrsDetRoughA}
  n64 = ScalarParameter() {defaultValue=1, parameterName=Terrain-GrsDetRoughB}
  n65 = LinearInterpolate(
   A=MaterialFunctionCall(
    TextureObject=TextureObject() {samplerType=SAMPLERTYPE_Normal, texture=/Game/Athena/Apollo/Environments/Landscape/Materials/Biome/CommonMaterials/T_CommonGrass_N}
    TextureSize=ScalarParameter() {defaultValue=3500, parameterName=GrsScle2test}
   ) {materialFunction=&WorldAlignedNormal}.Z Texture
   B=Constant3Vector() {constant=<rgba 0 0 1 0>}
   Alpha=ScalarParameter() {defaultValue=0.88, parameterName=Grass-NormalAmount2}
  )
  n66 = SetMaterialAttributes!(
   MaterialAttributes
   BaseColor=@n55
   Specular=FunctionOutput(None=VirtualTextureFeatureSwitch(No=@n62, Yes=@n62)) {outputName=TerrainSpec, sortPriority=2}
   Roughness=FunctionOutput(None=LinearInterpolate(A=@n63, B=@n64, Alpha=@n60.B)) {outputName=TerrainRough, sortPriority=3}
   Normal=FunctionOutput(None=@n65) {outputName=Normals, sortPriority=4}
  ) {attributeSetTypes [
    #1
    #2
    #3
    #4
   ]}
  FunctionOutput(None=@n66) {outputName=Material, sortPriority=-10}
  FunctionOutput(None=@n3.A) {outputName=Mobile-Height, sortPriority=1}
  FunctionOutput(None=Transform(None=@n65)) {outputName=WorldNormals, sortPriority=4}
  FunctionOutput(None=@n3.G) {outputName=BasecolorDetail, sortPriority=1}
  n67 = FunctionInput() {bUsePreviewValueAsDefault=true, inputName=CliffMask, inputType=FunctionInput_Scalar, previewValue=<1 1 1 1>}
  n68 = ScalarParameter() {defaultValue=3, parameterName=Grass-ClfMaskBarePow}
  n69 = Saturate(None=Multiply(A=Power(Base=@n67, Exp=@n68), B=ScalarParameter() {defaultValue=6, parameterName=Grass-ClfMaskBareMod}))
  n70 = ScalarParameter() {defaultValue=0.75, parameterName=Grass-HeightLerpA}
  n71 = ScalarParameter() {parameterName=Grass-HeightLerpB}
  n72 = LinearInterpolate(
   A=Multiply(A=@n3.A, B=OneMinus(None=@n69))
   B=Saturate(None=LinearInterpolate(A=@n70, B=@n71, Alpha=@n23.A))
   Alpha=ScalarParameter() {defaultValue=0.5, parameterName=Grass-ClfMaskBareLERP}
  )
  FunctionOutput(None=@n72) {outputName=Heightmap, sortPriority=-9}
  FunctionOutput(None=@n23.A) {outputName=GrassColorMaskC, sortPriority=11}
  FunctionOutput(None=@n24) {outputName=GrassColorMaskA2, sortPriority=9}
  FunctionOutput(None=@n62) {outputName=SpecTexture-NoRVT, sortPriority=6}
  n73 = Divide(A=@n37, B=ScalarParameter() {defaultValue=3, parameterName=RVT_MipDistanceMidEffect})
  Multiply(A=@n62, B=Max(A=ScalarParameter() {defaultValue=0.4, parameterName=GrsNearSpecMin}, B=Saturate(None=@n73)))
  Multiply(A=@n1, B=Constant() {r=-1})
  n74 = CameraPositionWS()
  n75 = WorldPosition()
  n76 = ScalarParameter() {defaultValue=2500, parameterName=Grass-DistanceMidEffect}
  n77 = Constant()
  Multiply(A=@n62, B=OneMinus(None=SphereMask(A=@n74, B=@n75.XYZ, Radius=@n76, Hardness=@n77)))
  n78 = VectorParameter() {parameterName=TestNewColors, defaultValue=<rgba 0.1157 0.15 0.0332 1>}
  Reroute(None=LinearInterpolate(A=Multiply(A=@n78.RGB, B=@n40), B=@n41, Alpha=ScalarParameter() {defaultValue=1, parameterName=tersttw43w4te4t}))
  ScalarParameter() {defaultValue=100000, parameterName=DistanceFarEffect}
  ScalarParameter() {defaultValue=20000, parameterName=DistanceNearEffect}
  n79 = Constant() {r=2}
  n80 = Saturate(None=Multiply(A=Power(Base=@n67, Exp=@n79), B=Constant() {r=12}))
  StaticSwitch(True=Multiply(A=@n3.A, B=OneMinus(None=@n80)), False=@n72, Value=StaticBool() {value=true})
  n81 = Constant3Vector() {constant=<rgba 0 -1 1 0>}
  n82 = VertexNormalWS()
  n83 = Constant3Vector() {constant=<rgba 0 0 1 0>}
  n84 = Multiply(
   A=Constant3Vector() {constant=<rgba 1 0 0 0>}
   B=If(A=DotProduct(A=@n82, B=@n83) {desc=generate sign for this projection}, B=Constant(), "A > B"=Constant() {r=1}, "A < B"=Constant() {r=-1})
  )
  n85 = WorldPosition() {worldPositionShaderOffset=WPT_ExcludeAllShaderOffsets}
  n86 = Multiply(A=Abs(None=ScalarParameter() {defaultValue=1600, group=Grass, parameterName=Grass-NormalScale})) {constA=-1, constB=-1}
  n87 = TextureSample(UVs=ComponentMask(None=Divide(A=@n85.XYZ, B=@n86) {constB=-512}) {g=true, r=true}) {
   samplerSource=SSM_Wrap_WorldGroupSettings
   samplerType=SAMPLERTYPE_Normal
   texture=&T_CommonGrass_N
  }
  n88 = VectorParameter() {parameterName=GrassNormalAmount-Mod, defaultValue=<rgba 0.15 0.15 1 1>}
  n89 = MaterialFunctionCall!(Vector1=@n82, Vector2=Constant3Vector() {constant=<rgba 0 1 0 0>}) {materialFunction=/Engine/Functions/Engine_MaterialFunctions02/Math/CreateThirdOrthogonalVector}
  n90 = MaterialFunctionCall(
   VectorToTransform=Multiply(A=Add(A=@n81, B=@n84), B=Multiply(A=@n87.RGB, B=@n88.RGB))
   BasisX=@n89.Vector3
   BasisY=@n89.Vector3
   BasisZ=@n89.Vector3
  ) {
   desc=lerps in when the vertex normal faces X or Negative X
   materialFunction=&Transform3x3Matrix
  }
  Transform(None=@n90.Transformed Vector) {transformSourceType=TRANSFORMSOURCE_World, transformType=TRANSFORM_Tangent}
 ]}
