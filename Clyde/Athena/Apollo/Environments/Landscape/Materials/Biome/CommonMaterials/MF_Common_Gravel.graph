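# Gravel landscape layer material function for Apollo terrain biomes, blending tiled dirt and gravel textures.
# Layer weight times a variance mask blends color and normals; layer weight also fades out a darkening tint.
# Chained MF_LS_DivideColor calls recolor for Mountain, Swamp and Farm using MF_ApolloTerrainMasks.
# Outputs material attributes and separate channels; roughness reuses the dirt specular channel, and the gravel mask texture is unconnected.
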MaterialFunction /Clyde/Athena/Apollo/Environments/Landscape/Materials/Biome/CommonMaterials/MF_Common_Gravel
graph {nodes [
  Constant() {r=1.85}
  n1 = MaterialFunctionCall!(UseLandscapeCoords=FunctionInput(Preview=StaticBool() {value=true}) {bUsePreviewValueAsDefault=true, inputName=UseLandscapeCoords, inputType=FunctionInput_StaticBool}) {materialFunction=&/MF_ApolloTerrainMasks}
  n2 = StaticBool() {value=true}
  n3 = Constant3Vector() {constant=<rgba 1 0.7337 0.5061 0>}
  n4 = Constant3Vector() {constant=<rgba 0.6878 0.712 0.724 1>}
  n5 = LandscapeLayerCoords() {mappingScale=4.5}
  n6 = TextureSample(UVs=@n5) {samplerSource=SSM_Wrap_WorldGroupSettings, texture=/Game/Environments/Apollo/Sets/Terrain/Textures/T_Apollo_Dirt_D_01}
  n7 = Constant4Vector() {constant=<rgba 0.7 0.7 0.7 1>}
  n8 = LandscapeLayerCoords() {mappingScale=4}
  n9 = TextureSample(UVs=@n8) {texture=/Game/Environments/Apollo/Sets/Terrain/Textures/T_Apollo_Gravel_D_01}
  n10 = FunctionInput() {inputName=LayerWeight, inputType=FunctionInput_Scalar}
  n11 = FunctionInput() {bUsePreviewValueAsDefault=true, inputName=VarianceMask, inputType=FunctionInput_Scalar, previewValue=<1 1 1 1>}
  n12 = ScalarParameter() {defaultValue=1.65, parameterName=Gravel-VarMod}
  n13 = Multiply(A=@n10, B=Saturate(None=Multiply(A=@n11, B=@n12)))
  n14 = LinearInterpolate(A=@n6.RGBA, B=Multiply(A=Max(A=@n7.RGBA, B=@n9.A), B=@n9.RGBA), Alpha=@n13)
  n15 = Multiply(A=Constant3Vector() {constant=<rgba 0.9 1 1 0>}, B=ComponentMask(None=@n14) {b=true, g=true, r=true})
  n16 = LinearInterpolate(A=Multiply(A=Constant3Vector() {constant=<rgba 0.45 0.3773 0.225 1>}, B=@n15), B=@n15, Alpha=@n10)
  n17 = MaterialFunctionCall!(
   UseBiomeColorMask=StaticBool() {value=true}
   BiomeColorMask=@n1.Farm
   BiomeColorDiv=Constant3Vector() {constant=<rgba 1 1 1 0>}
   BiomeColor=Multiply(A=Constant3Vector() {constant=<rgba 0.4271 0.3408 0.1062 1>}) {constB=4}
   Basecolor=MaterialFunctionCall(
    UseBiomeColorMask=StaticBool() {value=true}
    BiomeColorMask=@n1.Swamp
    BiomeColorDiv=Constant3Vector() {constant=<rgba 0.8 0.5869 0.4049 0>}
    BiomeColor=Constant3Vector() {constant=<rgba 0.3465 0.3481 0.35 1>}
    Basecolor=MaterialFunctionCall(UseBiomeColorMask=@n2, BiomeColorMask=@n1.Mountain, BiomeColorDiv=@n3, BiomeColor=@n4, Basecolor=@n16) {materialFunction=&/MF_LS_DivideColor}.Result
   ) {materialFunction=&MF_LS_DivideColor}.Result
  ) {materialFunction=&MF_LS_DivideColor}
  n18 = TextureSample(UVs=@n5) {
   samplerSource=SSM_Wrap_WorldGroupSettings
   samplerType=SAMPLERTYPE_Masks
   texture=/Game/Environments/Apollo/Sets/Terrain/Textures/T_Apollo_Dirt_S_01
  }
  n19 = TextureSample(UVs=@n5) {
   samplerSource=SSM_Wrap_WorldGroupSettings
   samplerType=SAMPLERTYPE_Normal
   texture=/Game/Environments/Apollo/Sets/Terrain/Textures/T_Apollo_Dirt_N_01
  }
  n20 = Multiply(A=@n19.RGB, B=Constant3Vector() {constant=<rgba 1.5 1.5 1 1>})
  n21 = SetMaterialAttributes!(
   MaterialAttributes
   BaseColor=@n17.Result
   Specular=@n18.R
   Roughness=@n18.R
   Normal=LinearInterpolate(
    A=@n20
    B=TextureSample(UVs=@n8) {samplerType=SAMPLERTYPE_Normal, texture=/Game/Environments/Apollo/Sets/Terrain/Textures/T_Apollo_Gravel_N_01}.RGB
    Alpha=@n13
   )
  ) {attributeSetTypes [
    #1
    #2
    #3
    #4
   ]}
  FunctionOutput(None=@n21) {outputName=Material}
  FunctionOutput(None=@n6.A) {outputName=Height, sortPriority=1}
  FunctionOutput(None=@n17.Result) {outputName=Diffuse, sortPriority=1}
  FunctionOutput(None=@n18.R) {outputName=Spec, sortPriority=1}
  FunctionOutput(None=@n18.B) {outputName=Rough, sortPriority=1}
  FunctionOutput(None=@n20) {outputName=Normals, sortPriority=1}
  n22 = ScalarParameter() {defaultValue=1.85, parameterName=Gravel-BiomeHeightMaskMod}
  n23 = OneMinus(None=FunctionOutput(None=ComponentMask(None=@n14) {a=true}) {outputName=Height-W/Variance, sortPriority=1})
  n24 = FunctionInput(Preview=Constant() {r=1}) {bUsePreviewValueAsDefault=true, inputName=BiomeColorMask, inputType=FunctionInput_Scalar, sortPriority=1}
  Multiply(A=Saturate(None=Multiply(A=@n22, B=@n23)), B=@n24)
  n25 = FunctionInput(Preview=StaticBool()) {bUsePreviewValueAsDefault=true, inputName=UseBiomeColorMask, inputType=FunctionInput_StaticBool, previewValue=<64 64 64 6>}
  Reroute(None=MaterialFunctionCall(
    UseBiomeColorMask=@n25
    BiomeColorMask=@n24
    BiomeColorDiv=FunctionInput() {bUsePreviewValueAsDefault=true, inputName=BiomeColorDiv, sortPriority=2}
    BiomeColor=FunctionInput() {bUsePreviewValueAsDefault=true, inputName=BiomeColor, sortPriority=3}
    Basecolor=@n16
   ) {materialFunction=&MF_LS_DivideColor}.Result)
  Reroute(None=@n6.RGB) {desc=Diffuse}
  TextureSample() {samplerType=SAMPLERTYPE_Masks, texture=/Game/Environments/Apollo/Sets/Terrain/Textures/T_Apollo_Gravel_S_01}
 ]}
