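# Material function in the Apollo landscape biome common materials that declares a shared set of landscape layers.
# A single LandscapeLayerBlend node defines weight-blended Gravel, Mud, Sand, Asphalt and BiomeA through BiomeD layers and feeds the only function output.
# No layer or height inputs are connected and all constants are zero, so it carries layer names rather than any surface shading.
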MaterialFunction /Clyde/Athena/Apollo/Environments/Landscape/Materials/Biome/CommonMaterials/MF_Common_Landscape_Layers
graph {nodes [
  n1 = LandscapeLayerBlend!(Layer Gravel, Layer Mud, Layer Sand, Layer Asphalt, Layer BiomeA, Layer BiomeB, Layer BiomeC, Layer BiomeD) {layers [
    {
     blendType=LB_WeightBlend
     constHeightInput=0
     layerName=Gravel
     previewWeight=0
     constLayerInput=<0 0 0>
     heightInput {expression=null, inputName=None, mask=0, maskA=0, maskB=0, maskG=0, maskR=0, outputIndex=0}
     layerInput {expression=null, inputName=None, mask=0, maskA=0, maskB=0, maskG=0, maskR=0, outputIndex=0}
    }
    {
     blendType=LB_WeightBlend
     constHeightInput=0
     layerName=Mud
     previewWeight=0
     constLayerInput=<0 0 0>
     heightInput {expression=null, inputName=None, mask=0, maskA=0, maskB=0, maskG=0, maskR=0, outputIndex=0}
     layerInput {expression=null, inputName=None, mask=0, maskA=0, maskB=0, maskG=0, maskR=0, outputIndex=0}
    }
    {
     blendType=LB_WeightBlend
     constHeightInput=0
     layerName=Sand
     previewWeight=0
     constLayerInput=<0 0 0>
     heightInput {expression=null, inputName=None, mask=0, maskA=0, maskB=0, maskG=0, maskR=0, outputIndex=0}
     layerInput {expression=null, inputName=None, mask=0, maskA=0, maskB=0, maskG=0, maskR=0, outputIndex=0}
    }
    {
     blendType=LB_WeightBlend
     constHeightInput=0
     layerName=Asphalt
     previewWeight=0
     constLayerInput=<0 0 0>
     heightInput {expression=null, inputName=None, mask=0, maskA=0, maskB=0, maskG=0, maskR=0, outputIndex=0}
     layerInput {expression=null, inputName=None, mask=0, maskA=0, maskB=0, maskG=0, maskR=0, outputIndex=0}
    }
    {
     blendType=LB_WeightBlend
     constHeightInput=0
     layerName=BiomeA
     previewWeight=0
     constLayerInput=<0 0 0>
     heightInput {expression=null, inputName=None, mask=0, maskA=0, maskB=0, maskG=0, maskR=0, outputIndex=0}
     layerInput {expression=null, inputName=None, mask=0, maskA=0, maskB=0, maskG=0, maskR=0, outputIndex=0}
    }
    {
     blendType=LB_WeightBlend
     constHeightInput=0
     layerName=BiomeB
     previewWeight=0
     constLayerInput=<0 0 0>
     heightInput {expression=null, inputName=None, mask=0, maskA=0, maskB=0, maskG=0, maskR=0, outputIndex=0}
     layerInput {expression=null, inputName=None, mask=0, maskA=0, maskB=0, maskG=0, maskR=0, outputIndex=0}
    }
    {
     blendType=LB_WeightBlend
     constHeightInput=0
     layerName=BiomeC
     previewWeight=0
     constLayerInput=<0 0 0>
     heightInput {expression=null, inputName=None, mask=0, maskA=0, maskB=0, maskG=0, maskR=0, outputIndex=0}
     layerInput {expression=null, inputName=None, mask=0, maskA=0, maskB=0, maskG=0, maskR=0, outputIndex=0}
    }
    {
     blendType=LB_WeightBlend
     constHeightInput=0
     layerName=BiomeD
     previewWeight=0
     constLayerInput=<0 0 0>
     heightInput {expression=null, inputName=None, mask=0, maskA=0, maskB=0, maskG=0, maskR=0, outputIndex=0}
     layerInput {expression=null, inputName=None, mask=0, maskA=0, maskB=0, maskG=0, maskR=0, outputIndex=0}
    }
   ]}
  FunctionOutput(None=@n1)
 ]}
