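# Landscape material function that builds a common mud surface from texture and biome inputs, outputting material attributes and a height.
# Diffuse is scaled and tinted brown, then passed through MF_LS_DivideColor for optional masked biome coloring.
# A HeightLerp driven by the squared inverse variance mask darkens color, lowers roughness and flattens normals where its alpha is low.
# The height output is the input height inverted and compressed into a mid range.
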MaterialFunction /Clyde/Athena/Apollo/Environments/Landscape/Materials/Biome/CommonMaterials/MF_Common_Mud
graph {nodes [
  n1 = FunctionInput() {bUsePreviewValueAsDefault=true, inputName=VarianceMask, inputType=FunctionInput_Scalar, previewValue=<1 1 1 1>}
  n2 = OneMinus(None=@n1)
  n3 = FunctionInput() {bUsePreviewValueAsDefault=true, inputName=Height, inputType=FunctionInput_Scalar, previewValue=<1 1 1 1>}
  n4 = MaterialFunctionCall(Transition Phase=Multiply(A=@n2, B=@n2), Height Texture=@n3) {materialFunction=&HeightLerp}
  n5 = Add(A=@n4.Lerp Alpha No Contrast, B=Constant() {r=0.8})
  n6 = FunctionInput() {bUsePreviewValueAsDefault=true, inputName=Rough, inputType=FunctionInput_Scalar}
  n7 = FunctionInput(Preview=Constant() {r=1}) {bUsePreviewValueAsDefault=true, inputName=BiomeRoughMod, inputType=FunctionInput_Scalar, sortPriority=4}
  n8 = Constant3Vector() {constant=<rgba 0.25 0.25 1 0>}
  n9 = FunctionInput() {bUsePreviewValueAsDefault=true, inputName=Normals, previewValue=<0 0 1 1>}
  n10 = FunctionInput(Preview=StaticBool()) {bUsePreviewValueAsDefault=true, inputName=UseBiomeColorMask, inputType=FunctionInput_StaticBool, previewValue=<64 64 64 6>}
  n11 = FunctionInput(Preview=Constant() {r=1}) {bUsePreviewValueAsDefault=true, inputName=BiomeColorMask, inputType=FunctionInput_Scalar, sortPriority=1}
  n12 = FunctionInput() {bUsePreviewValueAsDefault=true, inputName=BiomeColorDiv, sortPriority=2, previewValue=<1 1 1 1>}
  n13 = FunctionInput() {bUsePreviewValueAsDefault=true, inputName=BiomeColor, sortPriority=3, previewValue=<1 1 1 1>}
  n14 = Multiply(
   A=Constant3Vector() {constant=<rgba 0.3854 0.3231 0.1927 1>}
   B=Multiply(A=FunctionInput() {bUsePreviewValueAsDefault=true, inputName=Diffuse}) {constB=3.9}
  )
  n15 = Constant() {r=0.7}
  n16 = SetMaterialAttributes!(
   MaterialAttributes
   BaseColor=Multiply(
    A=MaterialFunctionCall!(UseBiomeColorMask=@n10, BiomeColorMask=@n11, BiomeColorDiv=@n12, BiomeColor=@n13, Basecolor=@n14) {materialFunction=&/MF_LS_DivideColor}.Result
    B=LinearInterpolate(A=@n15, Alpha=@n4.Lerp Alpha No Contrast)
   )
   Specular=FunctionInput() {bUsePreviewValueAsDefault=true, inputName=Spec, inputType=FunctionInput_Scalar}
   Roughness=Multiply(A=Saturate(None=@n5), B=Multiply(A=@n6, B=@n7))
   Normal=LinearInterpolate(A=Multiply(A=@n8, B=@n9), B=@n9, Alpha=@n4.Lerp Alpha No Contrast)
  ) {attributeSetTypes [
    #1
    #2
    #3
    #4
   ]}
  FunctionOutput(None=@n16) {outputName=Material}
  n17 = OneMinus(None=@n3)
  n18 = Constant() {r=2}
  FunctionOutput(None=Add(A=Divide(A=@n17, B=@n18), B=Constant() {r=0.25}) {constB=0.25}) {outputName=Height, sortPriority=1}
 ]}
