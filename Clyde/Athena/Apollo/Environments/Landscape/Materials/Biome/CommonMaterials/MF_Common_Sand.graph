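# Landscape material function for Apollo biome sand, outputting shore sand, a simpler under-ocean variant and texture height.
# Shared sand color, normal and mask textures use landscape or scaled mesh UVs via static switch, with biome tinting through MF_LS_DivideColor.
# World-height gradients around Z -3909, broken up by height lerps and a world-space mask, vary shoreline darkening, specular, roughness and normal flattening.
# Quality switches simplify normals on low and medium.
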MaterialFunction /Clyde/Athena/Apollo/Environments/Landscape/Materials/Biome/CommonMaterials/MF_Common_Sand
graph {
 properties {description=-3909}
 nodes [
  Constant2Vector()
  n1 = StaticSwitch(
   True=Multiply(A=Constant() {r=1000}, B=TextureCoordinate())
   False=LandscapeLayerCoords() {mappingScale=7}
   Value=FunctionInput(Preview=StaticBool()) {bUsePreviewValueAsDefault=true, inputName=UseUnderOceanSandMaterial, inputType=FunctionInput_StaticBool, sortPriority=10}
  )
  n2 = TextureSample(UVs=@n1) {samplerSource=SSM_Wrap_WorldGroupSettings, texture=/Game/Environments/Apollo/Sets/Terrain/Textures/T_Apollo_Sand_D_01}
  n3 = VertexNormalWS()
  n4 = TextureSample(UVs=@n1) {
   samplerSource=SSM_Wrap_WorldGroupSettings
   samplerType=SAMPLERTYPE_Normal
   texture=/Game/Environments/Apollo/Sets/Terrain/Textures/T_Apollo_Sand_N_01
  }
  n5 = TextureSample(UVs=@n1) {
   samplerSource=SSM_Wrap_WorldGroupSettings
   samplerType=SAMPLERTYPE_Masks
   texture=/Game/Environments/Apollo/Sets/Terrain/Textures/T_Apollo_Sand_S_01
  }
  n6 = Multiply(A=@n5.B, B=Constant() {r=1.75})
  n7 = Constant() {r=-1.5}
  n8 = Multiply(A=Saturate(None=Add(A=@n6, B=@n7)), B=@n5.R)
  n9 = SetMaterialAttributes!(
   MaterialAttributes
   BaseColor=Multiply(A=@n2.RGB, B=Constant() {r=0.45})
   Specular=@n8
   Roughness=Constant() {r=0.595}
   Normal=QualitySwitch(Default=@n4.RGB, Low=@n3, High=@n4.RGB, Medium=@n3)
  ) {attributeSetTypes [
    #1
    #2
    #3
    #4
   ]}
  FunctionOutput(None=@n9) {outputName=UnderOceanSandMaterial}
  n10 = FunctionInput(Preview=StaticBool()) {bUsePreviewValueAsDefault=true, inputName=UseBiomeColorMask, inputType=FunctionInput_StaticBool, previewValue=<64 64 64 6>}
  n11 = FunctionInput(Preview=Constant() {r=1}) {bUsePreviewValueAsDefault=true, inputName=BiomeColorMask, inputType=FunctionInput_Scalar, sortPriority=1}
  n12 = FunctionInput() {inputName=BiomeColorDiv, sortPriority=2}
  n13 = FunctionInput() {inputName=BiomeColor, sortPriority=3}
  n14 = Add(A=ComponentMask(None=WorldPosition() {worldPositionShaderOffset=WPT_ExcludeAllShaderOffsets}.XYZ) {b=true}, B=Constant() {r=3909})
  n15 = Saturate(None=Divide(A=@n14, B=Constant() {r=125}))
  n16 = OneMinus(None=FunctionOutput(None=@n2.A) {outputName=Height})
  n17 = WorldPosition() {worldPositionShaderOffset=WPT_ExcludeAllShaderOffsets}
  n18 = TextureSample(UVs=Divide(A=ComponentMask(None=@n17.XYZ) {g=true, r=true}, B=Constant() {r=8192}) {constB=8192}) {
   samplerSource=SSM_Wrap_WorldGroupSettings
   samplerType=SAMPLERTYPE_Grayscale
   texture=/Game/Athena/Apollo/Environments/Landscape/Materials/Biome/CommonMaterials/T_ApolloGrass_ColorMask
  }
  n19 = MaterialFunctionCall!(Diffuse 1, Diffuse 2, Transistion Phase=@n15, Height Texture 1=@n16, Height Texture 2=@n18.R, Contrast) {materialFunction=/Engine/Functions/Engine_MaterialFunctions02/Texturing/HeightLerpWithTwoHeightMaps}
  n20 = Add(A=OneMinus(None=@n19.Lerp Alpha No Contrast), B=Constant() {r=0.5})
  n21 = Constant() {r=25}
  n22 = Constant() {r=4}
  n23 = MaterialFunctionCall(Transition Phase=Saturate(None=Divide(A=@n14, B=@n21)), Height Texture=Add(A=Multiply(A=@n16, B=@n22), B=Constant() {r=-0.5})) {materialFunction=&HeightLerp}
  n24 = Constant() {r=0.9}
  n25 = Multiply(
   A=Multiply(
    A=MaterialFunctionCall!(UseBiomeColorMask=@n10, BiomeColorMask=@n11, BiomeColorDiv=@n12, BiomeColor=@n13, Basecolor=@n2.RGB) {materialFunction=&/MF_LS_DivideColor}.Result
    B=Saturate(None=@n20)
   )
   B=Saturate(None=Add(A=@n23.Lerp Alpha No Contrast, B=@n24))
  )
  n26 = Constant() {r=0.3}
  n27 = Saturate(None=LinearInterpolate(A=Constant() {r=-0.5}, B=Constant() {r=1.2}, Alpha=@n5.B))
  n28 = Add(A=Multiply(A=OneMinus(None=@n15), B=Constant() {r=4.5}), B=Constant() {r=-3.9})
  n29 = Constant() {r=2}
  n30 = Saturate(None=Add(A=Multiply(A=@n19.Lerp Alpha No Contrast, B=@n29), B=Constant() {r=-1}))
  n31 = Add(A=ComponentMask(None=WorldPosition() {worldPositionShaderOffset=WPT_ExcludeAllShaderOffsets}.XYZ) {b=true}, B=Constant() {r=4100})
  n32 = Constant() {r=200}
  n33 = Multiply(A=@n30, B=Saturate(None=Divide(A=@n31, B=@n32)))
  n34 = SetMaterialAttributes(
   BaseColor=Multiply(A=Constant() {r=0.7}, B=Multiply(A=@n25) {constB=1.5})
   Specular=LinearInterpolate(A=@n8, B=@n5.R, Alpha=@n23.Lerp Alpha No Contrast)
   Roughness=Max(A=LinearInterpolate(A=@n26, B=@n27, Alpha=@n23.Lerp Alpha No Contrast), B=Saturate(None=@n28))
   Normal=LinearInterpolate(A=@n4.RGB, B=VertexNormalWS(), Alpha=QualitySwitch(Default=@n33, Low=@n30, High=@n33, Medium=@n30))
  ) {attributeSetTypes [
    #1
    #2
    #3
    #4
   ]}
  FunctionOutput(None=@n34) {outputName=Material}
  MaterialFunctionCall(UVs=WorldPosition().XY) {materialFunction=&RadialGradientExponential}
 ]
}
